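# Niagara effect for the CanvasPrint backpack cosmetic, with hanging drips, a purple fire mesh and two glow sprites attached to the skeletal mesh.
# Drips grow and swing on bone-pivoted pendulum constraints that react to owner velocity, switching connection state late in life.
# FireSpline passes a lagged position to its meshes; it and Glow pulse with a system timeline and audio reactivity.
# User.Variant darkens the fire and hides the glow, while front-end and time-of-day modifiers adjust brightness.
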NiagaraSystem /BRCosmetics/Accessories/FORT_Backpacks/Backpack_M_MED_CanvasPrint/FX/NS_CanvasPrint_Backpack
dependencies {
 usedDataInterfaces [
  &NiagaraDataInterfaceVectorCurve
  &NiagaraDataInterfaceArrayFloat
  &NiagaraDataInterfaceColorCurve
  &NiagaraDataInterfaceCurve
  &NiagaraDataInterfaceArrayFloat3
  &NiagaraDataInterfaceSkeletalMesh
 ]
 usedDynamicInputs [
  &VectorLength
  &Waveform
  /Niagara/DynamicInputs/Lerp/SmoothLerpOverTime_Float
  &RemapRange
  &Cosine
  &Divide_Float
  &SetBoolByIntComparison
 ]
 usedModules [
  &self:SystemSpawnScript.SystemScriptSource.SystemScriptGraph.NiagaraNodeAssignment_0.SetVariables_#1
  &self:SystemSpawnScript.SystemScriptSource.SystemScriptGraph.NiagaraNodeAssignment_1.SetVariables_#2
  &self:SystemSpawnScript.SystemScriptSource.SystemScriptGraph.NiagaraNodeAssignment_6.SetVariables_#3
  &self:SystemSpawnScript.SystemScriptSource.SystemScriptGraph.NiagaraNodeAssignment_4.SetVariables_#4
  &SystemState
  &self:SystemSpawnScript.SystemScriptSource.SystemScriptGraph.NiagaraNodeAssignment_2.SetVariables_#5
  &self:SystemSpawnScript.SystemScriptSource.SystemScriptGraph.NiagaraNodeAssignment_7.SetVariables_#6
  &self:SystemSpawnScript.SystemScriptSource.SystemScriptGraph.NiagaraNodeAssignment_8.SetVariables_#7
  &self:SystemSpawnScript.SystemScriptSource.SystemScriptGraph.NiagaraNodeAssignment_9.SetVariables_#8
  &self:SystemSpawnScript.SystemScriptSource.SystemScriptGraph.NiagaraNodeAssignment_3.SetVariables_#9
  &Timeline
  &self:SystemSpawnScript.SystemScriptSource.SystemScriptGraph.NiagaraNodeAssignment_5.SetVariables_#10
 ]
}
emitters {
 Drips {
  emitterUpdateScript [
   &EmitterState(
    Life Cycle Mode=::NewEnumerator1
    Loop Delay=&RandomRangeFloat(Minimum=0, Maximum=0.222)
    Loop Duration=&RandomRangeFloat(Minimum=1, Maximum=2)
    Recalculate Duration Each Loop=true
    UseLoopDelay=true
   )
   &SpawnRate(
    Spawn Probability=&MakeCustomFloatFromBool(Bool=$System.IsRunning, True Float=1, False Float=0.5)
    SpawnRate=&MakeCustomFloatFromBool(
     Bool=$User.bIsFrontEndPreview
     True Float=1.2
     False Float=&MakeCustomFloatFromBool(
      Bool=$User.IsSkydiving
      True Float=0
      False Float=&MakeCustomFloatFromBool(Bool=$System.IsRunning, True Float=2.2, False Float=1.5)
     )
    )
    Use Spawn Probability=true
   )
  ]
  particleSpawnScript [
   &FN_SetSpawnGroup()
   &InitializeParticle(
    Color=<rgba 1 1 1 10>
    Lifetime Max=3.3
    Lifetime Min=2.5
    Lifetime Mode=::NewEnumerator1
    Mesh Scale Mode=::NewEnumerator3
    Mesh Uniform Scale=4
    UsePositionOffset=false
    Write Material Random=true
   )
   /SetVariables(Particles.Scale=&Multiply_VectorByFloat(
     Vector=&VectorFromFloat(Value=&MakeCustomFloatFromBool(Bool=$User.bIsFrontEndPreview, True Float=0.07, False Float=0.076))
     Float=&SelectFloatFromArray(
      Direct Array Index=&SpawnGroup()
      Float Selection Array=DataInterface() {propertyValues {FloatData [0.7,0.7,0.62,0.7]}}
     )
    )) {isSetParameters=true}
   &CameraOffset(Camera Offset Amount=&MakeCustomFloatFromBool(Bool=$User.bIsFrontEndPreview, True Float=-1, False Float=0))
   &DynamicMaterialParameters(
    Index 0 Param 3=&FN_FrontEndFloatModifier(Gameplay=0, Front End=1)
    Module.Index0Param0Write=true
    Module.Index0Param1Write=true
    Module.Index0Param2Write=true
    Module.Index0Param3Write=false
   )
   &SkeletalMeshLocation(
    Bone / Socket Index=&Add_Integer(A=1, B=&SpawnGroup())
    Bone Sampling Mode=::NewEnumerator4
    MaskSpawnGroup=false
    Skeletal Mesh=$System.Initial.Skeletal Mesh
   ) {enabled=false, module=SkeletalMeshLocation001}
  ]
  particleUpdateScript [
   &ParticleState()
   &SampleSkeletalMesh(
    AdditionalYawPitchRoll=false
    Bone / Socket Index:NiagaraInt32=&Add_Integer(A=1, B=&SpawnGroup())
    Bone Sampling Mode:ENiagaraBoneSamplingMode=::NewEnumerator4
    Sampled Position Offset:Vector3f=&SelectVectorFromArray(
     Direct Array Index=&SpawnGroup()
     Vector Selection Array=DataInterface() {propertyValues {FloatData [<0 0 -0.8>, <0 0 -0.1>, <0 0 -0.3>, <0 0 -4>]}}
    )
    Skeletal Mesh:NiagaraDataInterfaceSkeletalMesh=$System.Skeletal Mesh
   )
   /SetVariables(Particles.Connection:NiagaraBool=&SetBoolByFloatComparison(A=$Particles.NormalizedAge, B=0.85)) {isSetParameters=true}
   /SetVariables(Particles.Color=&MultiplyLinearColor(
     Linear Color=&MakeCustomLinearColorFromBool(Bool=$User.bIsFrontEndPreview, True Color=<rgba 0.0938 0.0867 0.0786 1>, False Color=<rgba 0.4271 0.3441 0.2981 1>)
     Scale Factor=&MakeVector4_FromLinearColor(LinearColor=&MakeLinearColorFromVectorAndFloat(
       "Vector (RGB)"=&FN_FrontEndVector3Modifier(
        Gameplay=&MakeVectorFromLinearColorRGB(Color=&FNTimeOfDayColor(
          Night Color=<rgba 1 1 1 1>
          Morning Color=<rgba 0.7031 0.7031 0.7031 1>
          Day Color=<rgba 0.7031 0.7031 0.7031 1>
          Evening Color=<rgba 1 1 1 1>
         ))
        Front End=&MakeVectorFromLinearColorRGB(Color=<rgba 0.7031 0.7031 0.7031 1>)
       )
       "Float (Alpha)"=&FloatFromCurve(
        FloatCurve=DataInterface() {propertyValues {Curve {keys [{interpMode=RCIM_Cubic, tangentMode=RCTM_User, time=0.2, value=0}, {interpMode=RCIM_Cubic, tangentMode=RCTM_User, time=0.45, value=1}]}}}
        Scale Curve=&FN_FrontEndFloatModifier(Gameplay=&FNTimeOfDayScalar(Night Scale=1, Morning Scale=1, Evening Scale=1), Front End=0.27)
       )
      ))
    )) {isSetParameters=true}
   &ScaleMeshSize(Scale Factor=&VectorFromFloat(Value=&FloatFromCurve(FloatCurve=DataInterface() {propertyValues {Curve {keys [
          {arriveTangent=0.6088, interpMode=RCIM_Cubic, leaveTangent=0.6088, tangentMode=RCTM_User, time=0.15, value=0.4}
          {interpMode=RCIM_Cubic, tangentMode=RCTM_User, time=0.7, value=1}
          {arriveTangent=-0.0171, interpMode=RCIM_Cubic, leaveTangent=-0.0171, tangentMode=RCTM_User, time=0.95, value=1}
          {arriveTangent=-22.3504, interpMode=RCIM_Cubic, leaveTangent=-22.3504, tangentMode=RCTM_User, time=1, value=0}
         ]}}})))
   &ScaleMeshSize(Scale Factor=&MakeVector(
     X=1
     Y=1
     Z=&FloatFromCurve(FloatCurve=DataInterface() {propertyValues {Curve {keys [
          {arriveTangent=4.2114, interpMode=RCIM_Cubic, leaveTangent=4.2114, tangentMode=RCTM_User, time=0.2, value=1}
          {arriveTangent=0.5863, interpMode=RCIM_Cubic, leaveTangent=0.5863, tangentMode=RCTM_User, time=1, value=2.5}
         ]}}})
    )) {module=ScaleMeshSize001}
   /SetVariables(Particles.MeshOrientation=&FindQuatBetweenTwoVectors(
     From Vector=<0 0 -1.0499>
     To Vector=&MakeCustomVectorFromBool(
      Bool=$Particles.Connection
      True Vector=&NormalizeVector(
       Vector To Normalize=&SubtractVectorFromPosition(Position=$Particles.Position, Vector=$Particles.SampleSkeletalMesh.SampledPosition)
       Safe Normalize=true
       Safe Fallback Threshold=0
       Safe Fallback Vector=<0 0 0.001>
      )
      False Vector=&NormalizeVector(
       Vector To Normalize=&Lerp_Vector(
        A=&SubtractVectorFromPosition(Position=$Particles.Position, Vector=$Particles.SampleSkeletalMesh.SampledPosition)
        B=<0 0 -1>
        Alpha=&FloatFromCurve(FloatCurve=DataInterface() {propertyValues {Curve {keys [<at 0.7 0>,<at 0.8 1>]}}}, Scale Curve=0)
       )
       Safe Normalize=true
       Safe Fallback Threshold=0
       Safe Fallback Vector=<0 0 0.001>
      )
     )
    )) {isSetParameters=true}
   &GravityForce()
   &Drag(Drag=2, Use Linear Drag=true, Use Rotational Drag=true)
   &PendulumSetup(
    Module.Enable Spring Driver=false
    Pendulum Length:NiagaraFloat=&FloatFromCurve(
     FloatCurve=DataInterface() {propertyValues {Curve {keys [<at 0 0 RCIM_Cubic>, {arriveTangent=0.0714, interpMode=RCIM_Cubic, leaveTangent=0.0714, tangentMode=RCTM_User, time=0.7, value=1}]}}}
     Scale Curve=3
    )
    Pendulum Pivot:Vector3f=$Particles.SampleSkeletalMesh.SampledPosition
    "Spring Driven Constraint (Tightness)"=2
   )
   &AccelerationForce(Acceleration=&Multiply_VectorByFloat(
     Vector=&Multiply_Vector(A=$Engine.Owner.Velocity, B=<-1 -1 0>)
     Float=&FloatFromCurve(FloatCurve=DataInterface() {propertyValues {Curve {keys [{interpMode=RCIM_Cubic, tangentMode=RCTM_User, time=0.6, value=0}, {interpMode=RCIM_Cubic, tangentMode=RCTM_User, time=0.9, value=1}]}}})
    ))
   &SolveForcesAndVelocity(Clamp Velocity=false, Limit Acceleration=false)
   /SetVariables(Particles.PositionTemp:NiagaraPosition=&AddVectorToPosition(Position=$Particles.Position, Vector=<0 0 0>)) {isSetParameters=true}
   &ScaleMeshSizeBySpeed(Max Scale Factor=<1 1 0.3>, Min Scale Factor:Vector3f=<0.6 0.6 1>, Speed Threshold=300)
   &PendulumConstraint()
   /SetVariables(Particles.Position=&MakeCustomVectorFromBool(
     Bool=$Particles.Connection
     True Vector=Linked() {linkedVariable {name=Particles.PositionTemp, type=NiagaraPosition}}
     False Vector=Linked() {linkedVariable {name=Particles.Position, type=NiagaraPosition}}
    )) {isSetParameters=true}
   &CalculateAccurateVelocity()
   /SetVariables(Particles.PositionDrips:NiagaraPosition=&MakeCustomVectorFromBool(
     Bool=&SetBoolByFloatComparison(A=$Particles.NormalizedAge, B=0.85)
     True Vector=Linked() {linkedVariable {name=Particles.Position, type=NiagaraPosition}}
     False Vector=$Transient.PhysicsPendulumPivot
    )) {isSetParameters=true}
  ]
  rendererClasses [&NiagaraMeshRendererProperties]
  renderers [{rendererClass=&NiagaraMeshRendererProperties, rendererIndex=0}]
 }
 FireSpline {
  emitterUpdateScript [
   &EmitterState(
    Inactive Response=::NewEnumerator1
    Life Cycle Mode=::NewEnumerator1
    Loop Behavior=::NewEnumerator1
    Loop Duration Mode=::NewEnumerator1
    UseLoopDelay=false
   )
   &SpawnBurst_Instantaneous(Use Spawn Probability=false, UseLoopCountLimit=false)
  ]
  particleSpawnScript [
   &InitializeParticle(Mesh Scale=<0.6593 0.2 0.4>, Mesh Scale Mode=::NewEnumerator4, UsePositionOffset=false, Write Material Random=true)
   &InitialMeshOrientation(
    Mesh Axis To Orient=<0 0 0>
    Mesh Orientation Mode=::NewEnumerator2
    Orientation Vector=<1 0.1 0>
    Rotation=<0 0.25 0>
    Use Rotation Vector=true
   ) {module=InitialMeshOrientation001}
  ]
  particleUpdateScript [
   &SkeletalMeshLocation(Bone Sampling Mode=::NewEnumerator4, MaskSpawnGroup=false, Sampled Position Offset=<2.9277 -2.3221 0>, Skeletal Mesh=$System.Skeletal Mesh) {module=SkeletalMeshLocation002}
   &ParticleState(Kill Particles When Lifetime Has Elapsed=false, Let Infinitely Lived Particles Die When Emitter Deactivates=true)
   &ScaleMeshSize(Scale Factor=&VectorFromCurve(
     VectorCurve=DataInterface() {propertyValues {
       XCurve {keys [<at 0 1 RCIM_Cubic>,<at 0.2 1.05 RCIM_Cubic>,<at 1 1 RCIM_Cubic>]}
       YCurve {keys [<at 0 1 RCIM_Cubic>,<at 0.2 1 RCIM_Cubic>,<at 1 1 RCIM_Cubic>]}
       ZCurve {keys [<at 0 1 RCIM_Cubic>,<at 0.2 1.15 RCIM_Cubic>,<at 1 1 RCIM_Cubic>]}
      }}
     CurveIndex=$System.Timeline.Time
    )) {module=ScaleMeshSize002}
   &ScaleMeshSize(Scale Factor=&Multiply_Vector(A=&Multiply_VectorByFloat(Vector=<0.2 0.2 0.2>, Float=0.8), B=<1 1 1>))
   &ScaleMeshSize(Scale Factor=&Lerp_Vector(A=<0.8 0.8 1>, B=<0.85 0.85 1>, Alpha=$System.AudioReactivity)) {module=ScaleMeshSize001}
   &ScaleColor(
    Scale Alpha=&Multiply_Float(A=1, B=&Lerp_Float(A=3.4, B=2, Alpha=$User.Variant))
    Scale RGB=&Multiply_VectorByFloat(
     Vector=&MakeVectorFromLinearColorRGB(Color=&LerpLinearColors(StartColor=<rgba 0.2585 0.1206 0.526 1>, EndColor=<rgba 0 0 0 1>, LerpFactor=$User.Variant))
     Float=&FN_FrontEndFloatModifier(
      Gameplay=&Multiply_Float(A=0.3, B=&FNTimeOfDayScalar(Night Scale=0.8, Morning Scale=0.9, Evening Scale=0.9))
      Front End=0.3
     )
    )
    ScaleA=true
    ScaleRGB=true
   )
   &InterpolateOverTime(
    Rate Of Change:NiagaraFloat=&Multiply_Float(
     A=$System.Interpolate Rate
     B=&SelectFloatFromArray(Direct Array Index=&SpawnGroup(), Float Selection Array=DataInterface() {propertyValues {FloatData [1,1,2]}})
    )
    Target Vector Value:NiagaraPosition=$Particles.Position
    Variable Type=::NewEnumerator2
   )
   /SetVariables(Particles.Interpolate DY:Vector4f=&MakeFloat4FromVectorAndFloat("Vector (RGB)"=$Particles.InterpolateOverTime.InterpolateOverTime_V3_.Moving Average, "Float (Alpha)"=$System.Interpolate Internsity)) {isSetParameters=true}
   /SetVariables(Particles.Color1:LinearColor=&MakeLinearColorFromVectorAndFloat(
     "Vector (RGB)"=&MakeVectorFromLinearColorRGB(Color=&LerpLinearColors(
       StartColor=&MultiplyLinearColor(
        Linear Color=&ColorFromCurve(
         ColorCurve=DataInterface() {propertyValues {
           AlphaCurve {keys [<at 0 1 RCIM_Cubic>,<at 1 1 RCIM_Cubic>]}
           BlueCurve {keys [<at 0 4 RCIM_Cubic>,<at 0.2181 20>,<at 1 4 RCIM_Cubic>]}
           GreenCurve {keys [<at 0 0.0849 RCIM_Cubic>,<at 0.2181 0.4244>,<at 1 0.0849 RCIM_Cubic>]}
           RedCurve {keys [<at 0 0.9528 RCIM_Cubic>,<at 0.2181 4.7639>,<at 1 0.9528 RCIM_Cubic>]}
          }}
         CurveIndex=$System.Timeline.Time
        )
        Scale Factor=&MakeVector4_FromFloat(Float=&FN_FrontEndFloatModifier(
          Gameplay=&Multiply_Float(
           A=&Multiply_Float(A=1, B=&FNTimeOfDayScalar(Night Scale=0.6, Morning Scale=1, Evening Scale=1))
           B=1
          )
          Front End=0.9
         ))
       )
       EndColor=<rgba 0 0 0 1>
       LerpFactor=$User.Variant
      ))
     "Float (Alpha)"=1
    )) {isSetParameters=true}
   /SetVariables(Particles.Color2=&MakeLinearColorFromVectorAndFloat(
     "Vector (RGB)"=&MakeVectorFromLinearColorRGB(Color=&LerpLinearColors(
       StartColor=&MultiplyLinearColor(
        Linear Color=&ColorFromCurve(
         ColorCurve=DataInterface() {propertyValues {
           AlphaCurve {keys [<at 0 1 RCIM_Cubic>,<at 1 1 RCIM_Cubic>]}
           BlueCurve {keys [<at 0 18 RCIM_Cubic>,<at 0.2181 22>,<at 1 18 RCIM_Cubic>]}
           GreenCurve {keys [<at 0 1.5 RCIM_Cubic>,<at 0.2181 3.0455>,<at 1 1.3125 RCIM_Cubic>]}
           RedCurve {keys [<at 0 5.7052 RCIM_Cubic>,<at 0.2181 7.8763>,<at 1 5.5655 RCIM_Cubic>]}
          }}
         CurveIndex=$System.Timeline.Time
        )
        Scale Factor=&MakeVector4_FromFloat(Float=&Multiply_Float(
          A=1
          B=&FN_FrontEndFloatModifier(
           Gameplay=&Multiply_Float(A=0.65, B=&FNTimeOfDayScalar(Night Scale=1, Morning Scale=1, Evening Scale=1.3))
           Front End=0.7
          )
         ))
       )
       EndColor=<rgba 0 0 0 1>
       LerpFactor=$User.Variant
      ))
     "Float (Alpha)"=1
    )) {isSetParameters=true}
   &DynamicMaterialParameters(
    Index 0 Param 1=&Add_Float(
     A=&FloatFromCurve(
      FloatCurve=DataInterface() {propertyValues {Curve {keys [<at 0 0.6 RCIM_Cubic>,<at 0.2 1.1 RCIM_Cubic>,<at 1 0.6 RCIM_Cubic>]}}}
      CurveIndex=$System.Timeline.Time
     )
     B=&Lerp_Float(A=0, B=0.14, Alpha=$System.AudioReactivity)
    )
    Index 0 Param 2=&MakeCustomFloatFromBool(Bool=$System.IsRunning, True Float=1, False Float=0.7)
    Index 0 Param 3=&FloatFromCurve(
     FloatCurve=DataInterface() {propertyValues {Curve {keys [<at 0 6 RCIM_Cubic>,<at 0.2 10 RCIM_Cubic>,<at 1 6 RCIM_Cubic>]}}}
     CurveIndex=$System.Timeline.Time
    )
    Index 0 Param 4=&FloatFromCurve(FloatCurve=DataInterface() {propertyValues {Curve {keys [<at 0 0.6 RCIM_Cubic>,<at 1 -0.15 RCIM_Cubic>]}}}, CurveIndex=$System.Timeline.Time)
    Index 2 Param 1=&Multiply_Float(A=$System.AudioReactivity, B=1)
    Index 2 Param 2=3
    Index 2 Param 3=&Lerp_Float(A=1, B=0.55, Alpha=$User.Variant)
    Index 2 Param 4=&Lerp_Float(
     A=&FN_FrontEndFloatModifier(Gameplay=3, Front End=2)
     B=&FN_FrontEndFloatModifier(Gameplay=2, Front End=1.6)
     Alpha=$User.Variant
    )
    Index 3 Param 1=$User.Variant
    Module.Index0Param0Write=true
    Module.Index0Param1Write=true
    Module.Index0Param2Write=true
    Module.Index0Param3Write=true
    Module.Index2Param0Write=true
    Module.Index2Param1Write=true
    Module.Index2Param2Write=true
    Module.Index2Param3Write=true
    Module.Index3Param0Write=true
    Module.Index3Param1Write=true
    Module.Index3Param2Write=true
    Module.Index3Param3Write=true
    Write Parameter Index 2=true
    Write Parameter Index 3=true
   )
   /SetVariables(Particles.GlitchDynamic:Vector4f=&MakeVector4(X=$System.AudioReactivity, Y=0, Z=0, W=0)) {isSetParameters=true}
  ]
  rendererClasses [&NiagaraMeshRendererProperties]
  renderers [
   {rendererClass=&NiagaraMeshRendererProperties, rendererIndex=0}
   {rendererClass=&NiagaraMeshRendererProperties, rendererIndex=1}
   {rendererClass=&NiagaraMeshRendererProperties, rendererIndex=2}
  ]
 }
 Glow {
  emitterUpdateScript [
   &EmitterState(
    Inactive Response=::NewEnumerator1
    Life Cycle Mode=::NewEnumerator1
    Loop Behavior=::NewEnumerator1
    Loop Duration Mode=::NewEnumerator1
    UseLoopDelay=false
   )
   &SpawnBurst_Instantaneous(Spawn Count=2, Use Spawn Probability=false, UseLoopCountLimit=false)
  ]
  particleSpawnScript [
   &FN_SetSpawnGroup(Number of Spawn Groups=2)
   &InitializeParticle(
    Color=<rgba 0.4764 0.0424 2 1>
    Sprite Size Mode=::NewEnumerator3
    Uniform Sprite Size=15
    UsePositionOffset=false
    Write Material Random=true
   )
   &CameraOffset(Camera Offset Amount=2)
  ]
  particleUpdateScript [
   &ParticleState(Kill Particles When Lifetime Has Elapsed=false, Let Infinitely Lived Particles Die When Emitter Deactivates=true)
   &ScaleSpriteSize(
    Scale Sprite Size Mode=::NewEnumerator0
    Uniform Scale Factor=&Lerp_Float(
     A=0.8
     B=1.3
     Alpha=&ClampFloat(Float=&Add_Float(
       A=$System.AudioReactivity
       B=&FloatFromCurve(
        FloatCurve=DataInterface() {propertyValues {Curve {keys [
            <at 0 0 RCIM_Cubic>
            {interpMode=RCIM_Cubic, tangentMode=RCTM_User, time=0.2, value=1}
            {interpMode=RCIM_Cubic, tangentMode=RCTM_User, time=0.6, value=1}
            <at 0.8 0 RCIM_Cubic>
           ]}}}
        CurveIndex=$System.Timeline.Time
       )
      ))
    )
   )
   &SkeletalMeshLocation(
    Bone / Socket Index=5
    Bone Sampling Mode=::NewEnumerator4
    MaskSpawnGroup=false
    Sampled Position Offset=&FN_FrontEndVector3Modifier(
     Gameplay=&SelectVectorFromArray(
      Direct Array Index=&SpawnGroup()
      Vector Selection Array=DataInterface() {propertyValues {FloatData [<4.5 -19 -0.2>, <-4.5 -19 -0.2>]}}
     )
     Front End=&SelectVectorFromArray(
      Direct Array Index=&SpawnGroup()
      Vector Selection Array=DataInterface() {propertyValues {FloatData [<4.5 -18 -0.2>, <-4.5 -18 -0.2>]}}
     )
    )
    Skeletal Mesh=$System.SkeletalMesh
   )
   &ScaleColor(
    Scale Alpha=&FN_FrontEndFloatModifier(
     Gameplay=0.5
     Front End=&FloatFromCurve(FloatCurve=DataInterface() {propertyValues {Curve {keys [<at 0.2 0 RCIM_Cubic>,<at 1 1>]}}}, CurveIndex=$System.Age, Scale Curve=0.2)
    )
    Scale RGB=&VectorFromFloat(Value=&Multiply_Float(
      A=&Lerp_Float(
       A=1
       B=1.3
       Alpha=&ClampFloat(Float=&Add_Float(
         A=$System.AudioReactivity
         B=&FloatFromCurve(
          FloatCurve=DataInterface() {propertyValues {Curve {keys [
              <at 0 0 RCIM_Cubic>
              {interpMode=RCIM_Cubic, tangentMode=RCTM_User, time=0.2, value=1}
              {interpMode=RCIM_Cubic, tangentMode=RCTM_User, time=0.6, value=1}
              <at 0.8 0 RCIM_Cubic>
             ]}}}
          CurveIndex=$System.Timeline.Time
         )
        ))
      )
      B=&FNTimeOfDayScalar(Night Scale=0.66, Morning Scale=0.85, Evening Scale=0.83)
     ))
    ScaleA=true
    ScaleRGB=true
   )
   &ScaleSpriteSize(Scale Sprite Size Mode=::NewEnumerator0, Uniform Scale Factor=&Lerp_Float(A=1, B=0, Alpha=$User.Variant)) {module=ScaleSpriteSize001}
  ]
  rendererClasses [&NiagaraSpriteRendererProperties]
  renderers [{rendererClass=&NiagaraSpriteRendererProperties}]
 }
}
summary {
 systemName=NS_CanvasPrint_Backpack
 emitters [
  {emitterName=FireSpline, rendererClasses [&NiagaraMeshRendererProperties]}
  {emitterName=Drips, rendererClasses [&NiagaraMeshRendererProperties]}
  {emitterName=Glow, rendererClasses [&NiagaraSpriteRendererProperties]}
 ]
 userVariables [
  User.bIsFrontEnd {type=NiagaraBool}
  User.bIsFrontEndPreview {type=NiagaraBool}
  User.DebugReactivity {type=NiagaraBool}
  User.Emote begin {type=NiagaraBool}
  User.IsSkydiving {type=NiagaraBool}
  User.KillReactivityCount {type=NiagaraInt32}
  User.Variant {type=NiagaraFloat, defaultValue=0, description=""}
 ]
}
