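# Physics asset for the CanvasPrint cosmetic backpack, giving its small figure (head, shoulders, legs) capsule collision and jointed limits for physics motion.
# Five constraints form a star around Body_base, tying the head, both shoulders and both legs directly to it.
# Shoulders get the widest swing and legs less, while the head is held almost rigid.
# A small Base_drops_left capsule has no constraint.
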PhysicsAsset /BRCosmetics/Accessories/FORT_Backpacks/Backpack_M_MED_CanvasPrint/Meshes/M_MED_CanvasPrint_Pack_Physics
bodies {
 Base_drops_left_bp_jnt [{shapeType=Capsule, length=0.8395, radius=1.4634, shapeName=Base_drops_left_bp_jnt_capsule, rotation=<rot 90 0 0>}]
 Body_base_bp_jnt [{shapeType=Capsule, length=2.2502, radius=10.1807, center=<7.3895 -0.0218 0.0443>, rotation=<rot -0.1499 -90 -0.1748>}]
 UpperHead_bp_jnt [{shapeType=Capsule, length=3.4651, radius=10.6316, shapeName=UpperHead_bp_jnt_capsule, center=<7.5601 0 0>, rotation=<rot 90 0 0>}]
 leg_left_bp_jnt [{shapeType=Capsule, length=6.2389, radius=2.3672, shapeName=leg_left_bp_jnt_capsule, center=<5 0 0>, rotation=<rot 90 0 0>}]
 leg_right_bp_jnt [{shapeType=Capsule, length=6.8207, radius=2.7336, shapeName=leg_right_bp_jnt_capsule, center=<-5 0 0>, rotation=<rot 90 0 0>}]
 shldr_left_bp_jnt [{shapeType=Capsule, length=3.8861, radius=2.3504, shapeName=shldr_left_bp_jnt_capsule, center=<7 0 0>, rotation=<rot 0 77.6322 -85.2499>}]
 shldr_right_bp_jnt [{
   shapeType=Capsule
   length=3.8293
   radius=2.1601
   shapeName=shldr_right_bp_jnt_capsule
   center=<-8 0 0>
   rotation=<rot 0.0746 77.993 -87.2987>
  }]
}
constraints [
 {bone1Name=shldr_left_bp_jnt, bone2Name=Body_base_bp_jnt, swing1LimitDegrees=45, swing2LimitDegrees=55, twistLimitDegrees=10}
 {bone1Name=shldr_right_bp_jnt, bone2Name=Body_base_bp_jnt, swing1LimitDegrees=40, swing2LimitDegrees=60, twistLimitDegrees=10}
 {bone1Name=UpperHead_bp_jnt, bone2Name=Body_base_bp_jnt, swing1LimitDegrees=1, swing2LimitDegrees=1, twistLimitDegrees=1}
 {bone1Name=leg_left_bp_jnt, bone2Name=Body_base_bp_jnt, swing1LimitDegrees=15, swing2LimitDegrees=35, twistLimitDegrees=5}
 {bone1Name=leg_right_bp_jnt, bone2Name=Body_base_bp_jnt, swing1LimitDegrees=15, swing2LimitDegrees=35, twistLimitDegrees=5}
]
