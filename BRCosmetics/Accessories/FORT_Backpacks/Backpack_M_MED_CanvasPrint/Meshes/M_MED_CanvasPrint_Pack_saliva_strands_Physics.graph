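# Physics asset for the saliva strands on the M_MED_CanvasPrint backpack, defining capsule bodies and joint limits for the dangling drop bones.
# The back, middle and first right-forward drops constrain to a large Body_base capsule, while the left drop hangs from its own Base_drops_left anchor, which has no constraint of its own.
# The right-forward strand continues as a three-link chain with tight swing and twist limits, while the left and middle drops swing most widely.
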PhysicsAsset /BRCosmetics/Accessories/FORT_Backpacks/Backpack_M_MED_CanvasPrint/Meshes/M_MED_CanvasPrint_Pack_saliva_strands_Physics
bodies {
 Base_drops_left_bp_jnt [{shapeType=Capsule, length=0.8395, radius=1.4634, shapeName=Base_drops_left_bp_jnt_capsule, rotation=<rot 90 0 0>}]
 Body_base_bp_jnt [{shapeType=Capsule, length=2.2502, radius=10.1807, center=<7.3895 -0.0218 0.0443>, rotation=<rot -0.1499 -90 -0.1748>}]
 dyn_drops_back_bp_jnt [{
   shapeType=Capsule
   length=1.2271
   radius=1.1571
   shapeName=dyn_drops_back_bp_jnt_capsule
   center=<0.0077 0.1747 -0.4215>
   rotation=<rot 0 -97.5142 -89.6832>
  }]
 dyn_drops_bp_right_fwd_01_jnt [{
   shapeType=Capsule
   length=1.3867
   radius=1.1354
   shapeName=dyn_drops_bp_right_fwd_01_jnt_capsule
   center=<0.6785 -0.1495 -0.1256>
   rotation=<rot 90 0 0>
  }]
 dyn_drops_bp_right_fwd_02_jnt [{
   shapeType=Capsule
   length=0.505
   radius=0.505
   shapeName=dyn_drops_bp_right_fwd_02_jnt_capsule
   center=<0.8709 -0.0121 0.006>
   rotation=<rot 0 89.2023 -89.6058>
  }]
 dyn_drops_bp_right_fwd_03_jnt [{
   shapeType=Capsule
   length=0.505
   radius=0.505
   shapeName=dyn_drops_bp_right_fwd_03_jnt_capsule
   center=<1.1435 -0.0035 -0.0163>
   rotation=<rot 0 -91.8562 -89.6216>
  }]
 dyn_drops_left_bp_jnt [{
   shapeType=Capsule
   length=0.8312
   radius=1.1079
   shapeName=dyn_drops_left_bp_jnt_capsule
   center=<1.3342 -0.03 0.1785>
   rotation=<rot 0 88.5957 -81.3519>
  }]
 dyn_drops_middle_bp_jnt [{shapeType=Capsule, length=1.2797, radius=1.1155, shapeName=dyn_drops_middle_bp_jnt_capsule, center=<1.4194 0 0>, rotation=<rot 90 0 0>}]
}
constraints [
 {bone1Name=dyn_drops_back_bp_jnt, bone2Name=Body_base_bp_jnt, swing1LimitDegrees=15, swing2LimitDegrees=15, twistLimitDegrees=5}
 {bone1Name=dyn_drops_bp_right_fwd_01_jnt, bone2Name=Body_base_bp_jnt, swing1LimitDegrees=20, swing2LimitDegrees=20, twistLimitDegrees=5}
 {
  bone1Name=dyn_drops_bp_right_fwd_02_jnt
  bone2Name=dyn_drops_bp_right_fwd_01_jnt
  swing1LimitDegrees=15
  swing2LimitDegrees=10
  twistLimitDegrees=2
 }
 {
  bone1Name=dyn_drops_bp_right_fwd_03_jnt
  bone2Name=dyn_drops_bp_right_fwd_02_jnt
  swing1LimitDegrees=15
  swing2LimitDegrees=12
  twistLimitDegrees=5
 }
 {bone1Name=dyn_drops_left_bp_jnt, bone2Name=Base_drops_left_bp_jnt, swing1LimitDegrees=45, swing2LimitDegrees=45, twistLimitDegrees=10}
 {bone1Name=dyn_drops_middle_bp_jnt, bone2Name=Body_base_bp_jnt, swing1LimitDegrees=45, swing2LimitDegrees=45, twistLimitDegrees=5}
]
