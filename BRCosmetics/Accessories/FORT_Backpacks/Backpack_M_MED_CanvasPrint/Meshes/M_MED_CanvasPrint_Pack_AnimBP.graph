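# Anim Blueprint for the CanvasPrint backpack cosmetic, playing idle and elimination reaction animations.
# Separate front-end and in-game state machines feed two RigidBody passes (pack, saliva strands), then additive pack-bone offsets.
# In game, the EnemyPlayerEliminated input event triggers the `elimination` flag, which is cleared after evaluation.
# When front-end reactivity is enabled, delays repeatedly fake eliminations and call Reactivity on the backpack's part modifier.
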AnimBlueprint /BRCosmetics/Accessories/FORT_Backpacks/Backpack_M_MED_CanvasPrint/Meshes/M_MED_CanvasPrint_Pack_AnimBP
target_skeleton=&/M_MED_CanvasPrint_Pack_Skeleton
graphs [
 AnimGraph {nodes [
   AnimGraphNode_StateMachine_0 = FE Backpack States() {is=title AnimGraphNode_StateMachine}
   AnimGraphNode_StateMachine_1 = Backpack States() {is=title AnimGraphNode_StateMachine}
   n1 = Get bIsInFrontEnd()
   n2 = RigidBody!(
    ComponentPose=Local To Component(LocalPose=Blend Poses by bool(BlendPose_0=@AnimGraphNode_StateMachine_0, BlendPose_1=@AnimGraphNode_StateMachine_1, bActiveValue=@n1))
    AlphaCurveName:Name=DynamicsAlpha
    ExternalForce
   ) -> (Pose) {
    alphaInputType=Curve
    bOverrideWorldGravity=true
    overridePhysicsAsset=&/M_MED_CanvasPrint_Pack_Physics
    alphaBoolBlend {blendOption=Linear}
    componentAppliedLinearAccClamp=<500 500 500>
    componentLinearAccScale=<1 1 1>
    componentLinearVelScale=<1 1 1>
    overrideWorldGravity {z=-980}
    simSpaceSettings {maxAngularAcceleration=1250, maxAngularVelocity=11, maxLinearAcceleration=700, maxLinearVelocity=5, worldAlpha=1}
   }
   n3 = RigidBody!(ComponentPose=@n2, Alpha:float, ExternalForce) -> (Pose) {
    bOverrideWorldGravity=true
    overridePhysicsAsset=&/M_MED_CanvasPrint_Pack_saliva_strands_Physics
    alphaBoolBlend {blendOption=Linear}
    componentAppliedLinearAccClamp=<500 500 500>
    componentLinearAccScale=<1 1 1>
    componentLinearVelScale=<1 1 1>
    overrideWorldGravity {z=-980}
    simSpaceSettings {maxAngularAcceleration=1250, maxAngularVelocity=11, maxLinearAcceleration=700, maxLinearVelocity=5, worldAlpha=1}
   }
   n4 = "Transform (Modify) Bone - Bone: pack"!(ComponentPose=@n3, bAlphaBoolEnabled:bool, Translation="0,-3.2,0", Rotation="0,0,-1.5", Scale) -> (Pose) {
    alphaInputType=Bool
    rotationMode=BMM_Additive
    translationMode=BMM_Additive
    alphaBoolBlend {blendOption=Linear}
    boneToModify {boneName=pack}
    rotation {roll=-1.5}
    translation {y=-3.2}
   }
   n5 = "Transform (Modify) Bone - Bone: pack"(ComponentPose=@n4, Translation="0,-1.5,0.25", Rotation="0,0,-0.2") {
    alphaInputType=Bool
    rotationMode=BMM_Additive
    translationMode=BMM_Additive
    alphaBoolBlend {blendOption=Linear}
    boneToModify {boneName=pack}
    rotation {roll=-0.2}
    translation {y=-1.5, z=0.25}
   }
   Output Pose(Result=Component To Local(ComponentPose=@n5)) {layerGroup=None}
  ]}
 FE Backpack States {
  entry=FE_Idle
  path=AnimGraph.AnimGraphNode_StateMachine_0.FE Backpack States
  ids=id
  nodes [
   AnimStateNode_0 = FE_Idle() {is=title AnimStateNode}
   AnimStateNode_2 = FE_Reaction() {is=title AnimStateNode, stateEntered {guid=#1, notifyName=Entered_FE_Reaction}}
   AnimStateTransitionNode_7 = AnimStateTransitionNode()
   AnimStateTransitionNode_8 = AnimStateTransitionNode() {bAutomaticRuleBasedOnSequencePlayerInState=true}
  ]
  links [
   {
    dst=FE_Reaction
    rule=AnimGraph.AnimGraphNode_StateMachine_0.FE Backpack States.AnimStateTransitionNode_7.Transition
    src=FE_Idle
    via=AnimStateTransitionNode_7
   }
   {
    dst=FE_Idle
    rule=AnimGraph.AnimGraphNode_StateMachine_0.FE Backpack States.AnimStateTransitionNode_8.Transition
    src=FE_Reaction
    via=AnimStateTransitionNode_8
   }
  ]
 }
 FE_Idle {
  path=AnimGraph.AnimGraphNode_StateMachine_0.FE Backpack States.AnimStateNode_0.FE_Idle
  nodes [Output Animation Pose(Result=Sequence Player 'Backpack_CanvasPrint_Idle'() {sequence=/BRCosmetics/Accessories/FORT_Backpacks/Backpack_M_MED_CanvasPrint/Anim/Backpack_CanvasPrint_Idle})]
 }
 FE_Reaction {
  path=AnimGraph.AnimGraphNode_StateMachine_0.FE Backpack States.AnimStateNode_2.FE_Reaction
  nodes [
   n1 = Sequence Player 'Backpack_CanvasPrint_Reaction'() {bLoopAnimation=false, sequence=/BRCosmetics/Accessories/FORT_Backpacks/Backpack_M_MED_CanvasPrint/Anim/Backpack_CanvasPrint_Reaction}
   Output Animation Pose(Result=@n1)
  ]
 }
 Transition {
  path=AnimGraph.AnimGraphNode_StateMachine_0.FE Backpack States.AnimStateTransitionNode_7.Transition
  nodes [Result(bCanEnterTransition=Get elimination() -> (elimination:bool))]
 }
 Transition {path=AnimGraph.AnimGraphNode_StateMachine_0.FE Backpack States.AnimStateTransitionNode_8.Transition, nodes [Result()]}
 Backpack States {
  entry=Idle
  path=AnimGraph.AnimGraphNode_StateMachine_1.Backpack States
  ids=id
  nodes [
   AnimStateNode_0 = Idle()
   AnimStateNode_2 = Reaction() {is=title AnimStateNode}
   AnimStateTransitionNode_7 = AnimStateTransitionNode()
   AnimStateTransitionNode_8 = AnimStateTransitionNode() {bAutomaticRuleBasedOnSequencePlayerInState=true}
  ]
  links [
   {
    dst=Reaction
    rule=AnimGraph.AnimGraphNode_StateMachine_1.Backpack States.AnimStateTransitionNode_7.Transition
    src=Idle
    via=AnimStateTransitionNode_7
   }
   {
    dst=Idle
    rule=AnimGraph.AnimGraphNode_StateMachine_1.Backpack States.AnimStateTransitionNode_8.Transition
    src=Reaction
    via=AnimStateTransitionNode_8
   }
  ]
 }
 Idle {
  path=AnimGraph.AnimGraphNode_StateMachine_1.Backpack States.AnimStateNode_0.Idle
  nodes [Output Animation Pose(Result=Sequence Player 'Backpack_CanvasPrint_Idle'() {sequence=&Backpack_CanvasPrint_Idle})]
 }
 Reaction {
  path=AnimGraph.AnimGraphNode_StateMachine_1.Backpack States.AnimStateNode_2.Reaction
  nodes [
   n1 = Sequence Player 'Backpack_CanvasPrint_Reaction'() {bLoopAnimation=false, sequence=&Backpack_CanvasPrint_Reaction}
   Output Animation Pose(Result=@n1)
  ]
 }
 Transition {
  path=AnimGraph.AnimGraphNode_StateMachine_1.Backpack States.AnimStateTransitionNode_7.Transition
  nodes [Result(bCanEnterTransition=Get elimination())]
 }
 Transition {path=AnimGraph.AnimGraphNode_StateMachine_1.Backpack States.AnimStateTransitionNode_8.Transition, nodes [Result()]}
 EventGraph {nodes [
   Event OnCharacterCustomizationLoaded()
   > n1 = Branch(Condition=Get bShouldDisplayFrontendReactivity() -> (bShouldDisplayFrontendReactivity:bool))
   > Do Once(Reset=@n5.then)
   > n2 = Delay(Duration=4)
   Set elimination(execute=[@n6.then, @n2.then], elimination:bool=true) -> (Output_Get:bool)
   > n3 = Cast To FortPlayerPawn(Object=TryGetFortPawn())
   > n4 = Cast To B_Athena_PartModifer_CanvasPrint_Backpack!(
    execute
    Object=GetModifierForPartType!(self:Fort Player Pawn Object Reference=@n3, PartType:EFortCustomPartType Enum=Backpack) -> (ReturnValue:Custom Character Part Modifier Object Reference)
   ) -> (then, CastFailed, AsB Athena Part Modifer Canvas Print Backpack:B Athena Part Modifer Canvas Print Backpack Object Reference) {is=title K2Node_DynamicCast}
   > Reactivity!(execute, self:B Athena Part Modifer Canvas Print Backpack Object Reference=@n4, KillCount:int=1) -> (then)
   n5 = Set elimination(execute=@n1.else, elimination=false)
   AnimNotify_Entered_FE_Reaction() {is=title K2Node_Event}
   > Set elimination(elimination=false)
   > n6 = Delay(Duration=13.5)
   n7 = Event BlueprintPostEvaluateAnimation() {is=title K2Node_Event}
   Set elimination(execute=[@n7.then, @n8.then], elimination=false)
   n8 = AnimNotify_ReactiveReset() {is=title K2Node_Event}
   n9 = Event BlueprintProcessAnimInputEvent()
   > Sequence()
   > Branch(Condition=IsNamed(self=@n9.AnimInputEvent, InEventName=aie.EnemyPlayerEliminated))
   > Set elimination(elimination=true)
   Branch()
   > Delay(Duration=10)
   > Set elimination(elimination=true)
   > n10 = Cast To FortPlayerPawn(Object=TryGetFortPawn())
   > Cast To B_Athena_PartModifer_CanvasPrint_Backpack(Object=GetModifierForPartType(self=@n10, PartType=Backpack))
   TryGetPawnOwner()
   Event BlueprintUpdateAnimation()
   Event BlueprintInitializeAnimation()
  ]}
]
variables [elimination {type=boolean}]
